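# Physics asset for the F_MED_Typhoon_Armored female player cosmetic body, covering its dynamic strap and back clip bones on left and right sides.
# Each dyn bone has a small capsule constrained to its own _MSTR bone, whose near-zero capsule serves as an anchor point.
# Angular limits keep twist tight throughout, with main straps swinging widest, top and bottom straps less, and back clips narrowest on one axis.
# No other skeleton bones have bodies.
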PhysicsAsset /BRCosmetics/Characters/Player/Female/Medium/Bodies/F_MED_Typhoon_Armored/Meshes/F_MED_Typhoon_Armored_Physics
bodies {
 dyn_clips_bk_01_l [{shapeType=Capsule, length=1.7414, radius=1.0559, center=<0.3446 0 0>, rotation=<rot -90 0.0033 -0.0032>}]
 dyn_clips_bk_01_l_MSTR [{shapeType=Capsule, length=0.01, radius=0.01}]
 dyn_clips_bk_01_r [{shapeType=Capsule, length=1.7414, radius=1.0559, center=<0.3446 0 0>, rotation=<rot -90 0.0033 -0.0032>}]
 dyn_clips_bk_01_r_MSTR [{shapeType=Capsule, length=0.01, radius=0.01}]
 dyn_strap_01_l [{shapeType=Capsule, length=1.7414, radius=1.0559, center=<0.9872 0 0>, rotation=<rot -90 -0.0149 0.0149>}]
 dyn_strap_01_l_MSTR [{shapeType=Capsule, length=0.01, radius=0.01}]
 dyn_strap_01_r [{shapeType=Capsule, length=1.7414, radius=1.0559, center=<0.9872 0 0>, rotation=<rot -90 -0.0149 0.0149>}]
 dyn_strap_01_r_MSTR [{shapeType=Capsule, length=0.01, radius=0.01}]
 dyn_strap_bt_01_l [{shapeType=Capsule, length=1.7414, radius=1.0559, center=<0.3446 0 0>, rotation=<rot -90 0.0264 -0.0263>}]
 dyn_strap_bt_01_l_MSTR [{shapeType=Capsule, length=0.01, radius=0.01}]
 dyn_strap_bt_01_r [{shapeType=Capsule, length=1.7414, radius=1.0559, center=<0.3446 0 0>, rotation=<rot -90 0.0264 -0.0263>}]
 dyn_strap_bt_01_r_MSTR [{shapeType=Capsule, length=0.01, radius=0.01}]
 dyn_strap_tp_01_l [{shapeType=Capsule, length=1.7414, radius=1.0559, center=<0.3446 0 0>, rotation=<rot -90 0.0264 -0.0263>}]
 dyn_strap_tp_01_l_MSTR [{shapeType=Capsule, length=0.01, radius=0.01}]
 dyn_strap_tp_01_r [{shapeType=Capsule, length=1.7414, radius=1.0559, center=<0.3446 0 0>, rotation=<rot -90 0.0264 -0.0263>}]
 dyn_strap_tp_01_r_MSTR [{shapeType=Capsule, length=0.01, radius=0.01}]
}
constraints [
 {bone1Name=dyn_strap_01_l, bone2Name=dyn_strap_01_l_MSTR, swing1LimitDegrees=45, swing2LimitDegrees=45, twistLimitDegrees=5}
 {bone1Name=dyn_strap_01_r, bone2Name=dyn_strap_01_r_MSTR, swing1LimitDegrees=45, swing2LimitDegrees=45, twistLimitDegrees=5}
 {bone1Name=dyn_strap_tp_01_l, bone2Name=dyn_strap_tp_01_l_MSTR, swing1LimitDegrees=15, swing2LimitDegrees=15, twistLimitDegrees=5}
 {bone1Name=dyn_strap_bt_01_l, bone2Name=dyn_strap_bt_01_l_MSTR, swing1LimitDegrees=15, swing2LimitDegrees=15, twistLimitDegrees=5}
 {bone1Name=dyn_strap_tp_01_r, bone2Name=dyn_strap_tp_01_r_MSTR, swing1LimitDegrees=15, swing2LimitDegrees=15, twistLimitDegrees=5}
 {bone1Name=dyn_strap_bt_01_r, bone2Name=dyn_strap_bt_01_r_MSTR, swing1LimitDegrees=15, swing2LimitDegrees=15, twistLimitDegrees=5}
 {bone1Name=dyn_clips_bk_01_r, bone2Name=dyn_clips_bk_01_r_MSTR, swing1LimitDegrees=5, swing2LimitDegrees=15, twistLimitDegrees=5}
 {bone1Name=dyn_clips_bk_01_l, bone2Name=dyn_clips_bk_01_l_MSTR, swing1LimitDegrees=5, swing2LimitDegrees=15, twistLimitDegrees=5}
]
