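# Post-process AnimBP for the armored female Typhoon body cosmetic, adding secondary motion to the pose copied from its parent mesh.
# Weighted constraints anchor straps, clips and knife before two RigidBody nodes simulate belt and armor physics.
# Clavicle and upper-arm twist deltas drive shoulder-pad and arm corrective bones.
# Weapon and skydiving events set material "Hide Element 01" to 1 only with an equipped Typhoon-tagged pickaxe outside skydiving.
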AnimBlueprint /BRCosmetics/Characters/Player/Female/Medium/Bodies/F_MED_Typhoon_Armored/Meshes/F_MED_Typhoon_Armored_AnimBP
target_skeleton=&/F_MED_Typhoon_Armored_Skeleton
graphs [
 AnimGraph {nodes [
   n1 = "Constraint - dyn_strap_tp_01_l_MSTR (P)"!(
    ComponentPose=Local To Component(LocalPose=Copy Pose From Mesh() {bUseAttachedParent=true})
    Alpha=1
    ConstraintWeights_0=1
    ConstraintWeights_1:float=0
   ) -> (Pose) {
    alphaBoolBlend {blendOption=Linear}
    boneToModify {boneName=dyn_strap_tp_01_l_MSTR}
    constraintSetup [{targetBone {boneName=spine_04}}]
    constraintWeights [1,0]
   }
   n2 = "Constraint - dyn_strap_bt_01_l_MSTR (PPPP)"!(
    ComponentPose=@n1
    Alpha=1
    ConstraintWeights_0=0.031
    ConstraintWeights_1:float=0.0988
    ConstraintWeights_2:float=0.3064
    ConstraintWeights_3:float=0.5639
   ) -> (Pose) {
    alphaBoolBlend {blendOption=Linear}
    boneToModify {boneName=dyn_strap_bt_01_l_MSTR}
    constraintSetup [{targetBone {boneName=spine_01}}, {targetBone {boneName=spine_02}}, {targetBone {boneName=spine_03}}, {targetBone {boneName=spine_04}}]
    constraintWeights [1,1,1,1]
   }
   n3 = "Constraint - dyn_strap_tp_01_r_MSTR (P)"!(ComponentPose=@n2, Alpha=1, ConstraintWeights_0=1, ConstraintWeights_1:float=0) -> (Pose) {
    alphaBoolBlend {blendOption=Linear}
    boneToModify {boneName=dyn_strap_tp_01_r_MSTR}
    constraintSetup [{targetBone {boneName=spine_04}}]
    constraintWeights [1,0]
   }
   n4 = "Constraint - dyn_strap_bt_01_r_MSTR (PPPP)"!(
    ComponentPose=@n3
    Alpha=1
    ConstraintWeights_0=0.031
    ConstraintWeights_1:float=0.0988
    ConstraintWeights_2:float=0.3064
    ConstraintWeights_3:float=0.5639
   ) -> (Pose) {
    alphaBoolBlend {blendOption=Linear}
    boneToModify {boneName=dyn_strap_bt_01_r_MSTR}
    constraintSetup [{targetBone {boneName=spine_01}}, {targetBone {boneName=spine_02}}, {targetBone {boneName=spine_03}}, {targetBone {boneName=spine_04}}]
    constraintWeights [1,1,1,1]
   }
   n5 = "Constraint - dyn_clips_bk_01_l_MSTR (PPPPPPPP)"!(
    ComponentPose=@n4
    Alpha=1
    ConstraintWeights_0=0.2324
    ConstraintWeights_1:float=0.0118
    ConstraintWeights_2:float=0.0306
    ConstraintWeights_3:float=0.0539
    ConstraintWeights_4:float=0.3895
    ConstraintWeights_5:float=0.2617
    ConstraintWeights_6:float=0.0017
    ConstraintWeights_7:float=0.0185
   ) -> (Pose) {
    alphaBoolBlend {blendOption=Linear}
    boneToModify {boneName=dyn_clips_bk_01_l_MSTR}
    constraintSetup [
     {targetBone {boneName=clavicle_l}}
     {targetBone {boneName=spine_01}}
     {targetBone {boneName=spine_02}}
     {targetBone {boneName=spine_03}}
     {targetBone {boneName=spine_04}}
     {targetBone {boneName=spine_05}}
     {targetBone {boneName=upperarm_twist_01_l}}
     {targetBone {boneName=clavicle_r}}
    ]
    constraintWeights [1,1,1,1,1,1,1,1]
   }
   n6 = "Constraint - dyn_clips_bk_01_r_MSTR (PPPPPPPP)"!(
    ComponentPose=@n5
    Alpha=1
    ConstraintWeights_0=0.0185
    ConstraintWeights_1:float=0.0017
    ConstraintWeights_2:float=0.0118
    ConstraintWeights_3:float=0.0306
    ConstraintWeights_4:float=0.0539
    ConstraintWeights_5:float=0.3895
    ConstraintWeights_6:float=0.2617
    ConstraintWeights_7:float=0.2324
   ) -> (Pose) {
    alphaBoolBlend {blendOption=Linear}
    boneToModify {boneName=dyn_clips_bk_01_r_MSTR}
    constraintSetup [
     {targetBone {boneName=clavicle_l}}
     {targetBone {boneName=upperarm_twist_01_r}}
     {targetBone {boneName=spine_01}}
     {targetBone {boneName=spine_02}}
     {targetBone {boneName=spine_03}}
     {targetBone {boneName=spine_04}}
     {targetBone {boneName=spine_05}}
     {targetBone {boneName=clavicle_r}}
    ]
    constraintWeights [1,1,1,1,1,1,1,1]
   }
   n7 = "Transform (Modify) Bone - Bone: padhelper_01_l"(
    ComponentPose=@n6
    Alpha=Get shoulderAlpha_l() -> (shoulderAlpha_l:double)
    Translation="4.5,-0.000,-0.000"
    Rotation="0,0,0"
    Scale="0.5,1.000,1.000"
   ) {
    rotationMode=BMM_Additive
    rotationSpace=BCS_BoneSpace
    scaleMode=BMM_Additive
    scaleSpace=BCS_BoneSpace
    translationMode=BMM_Additive
    translationSpace=BCS_BoneSpace
    alphaBoolBlend {blendOption=Linear}
    boneToModify {boneName=padhelper_01_l}
    scale {x=0.5}
    translation {x=4.5}
   }
   n8 = "Transform (Modify) Bone - Bone: padhelper_01_r"(
    ComponentPose=@n7
    Alpha=Get shoulderAlpha_r() -> (shoulderAlpha_r:double)
    Translation="1.02,0.000,-0.000"
    Rotation="0,0,0"
    Scale="0.5,1.000,1.000"
   ) {
    rotationMode=BMM_Additive
    rotationSpace=BCS_BoneSpace
    scaleMode=BMM_Additive
    scaleSpace=BCS_BoneSpace
    translationMode=BMM_Additive
    translationSpace=BCS_BoneSpace
    alphaBoolBlend {blendOption=Linear}
    boneToModify {boneName=padhelper_01_r}
    scale {x=0.5}
    translation {x=1.02}
   }
   n9 = "Transform (Modify) Bone - Bone: upperarm_twist_01_r"(ComponentPose=@n8, Alpha=Get inflateAlpha_r() -> (inflateAlpha_r:double), Translation="0,0,0", Rotation="0,0,0", Scale="1,1,1.2") {
    rotationMode=BMM_Additive
    rotationSpace=BCS_BoneSpace
    scaleMode=BMM_Additive
    scaleSpace=BCS_BoneSpace
    translationMode=BMM_Additive
    translationSpace=BCS_BoneSpace
    alphaBoolBlend {blendOption=Linear}
    boneToModify {boneName=upperarm_twist_01_r}
    scale {z=1.2}
   }
   n10 = "Constraint - dyn_knife_01_r_MSTR (PPPP)"!(
    ComponentPose=@n9
    Alpha=1
    ConstraintWeights_0=0.1163
    ConstraintWeights_1:float=0.2006
    ConstraintWeights_2:float=0.0401
    ConstraintWeights_3:float=0.643
   ) -> (Pose) {
    alphaBoolBlend {blendOption=Linear}
    boneToModify {boneName=dyn_knife_01_r_MSTR}
    constraintSetup [{targetBone {boneName=spine_01}}, {targetBone {boneName=spine_02}}, {targetBone {boneName=spine_04}}, {targetBone {boneName=pelvis}}]
    constraintWeights [1,1,1,1]
   }
   n11 = RigidBody!(ComponentPose=@n10, Alpha, OverrideWorldGravity:Vector="0,0,-980", ExternalForce) -> (Pose) {
    bClampLinearTranslationLimitToRefPose=true
    bOverrideWorldGravity=true
    overridePhysicsAsset=/BRCosmetics/Characters/Player/Female/Medium/Bodies/F_MED_Typhoon/Meshes/F_MED_Typhoon_Belt_Physics
    simulationSpace=BaseBoneSpace
    alphaBoolBlend {blendOption=Linear}
    baseBoneRef {boneName=pelvis}
    componentAppliedLinearAccClamp=<1500 1500 1500>
    componentLinearAccScale=<0.3 0.3 0.45>
    componentLinearVelScale=<0.3 0.3 0.45>
    overrideWorldGravity {z=-980}
   }
   n12 = RigidBody(ComponentPose=@n11, OverrideWorldGravity="0,0,-980") {
    bClampLinearTranslationLimitToRefPose=true
    bOverrideWorldGravity=true
    overridePhysicsAsset=&/F_MED_Typhoon_Armored_Physics
    simulationSpace=BaseBoneSpace
    alphaBoolBlend {blendOption=Linear}
    baseBoneRef {boneName=spine_05}
    componentAppliedLinearAccClamp=<1500 1500 1500>
    componentLinearAccScale=<0.3 0.3 0.45>
    componentLinearVelScale=<0.3 0.3 0.45>
    overrideWorldGravity {z=-980}
   }
   Output Pose(Result=Component To Local(ComponentPose=@n12)) {layerGroup=None}
  ]}
 EventGraph {nodes [
   Event BlueprintUpdateAnimation()
   > Set inflateAlpha_r(inflateAlpha_r:double=GetAlphaFromReferenceTransformDelta(BoneName=upperarm_r, BaseName=clavicle_r, DeltaComponentType=QuaternionTwist, InputRange=(X=-90,Y=-130), TwistAxis="0,0,1")) -> (Output_Get:double)
   n1 = GetOwnerSkeletalMeshComponent()
   > Set shoulderAlpha_r(shoulderAlpha_r:double=GetAlphaFromReferenceTransformDelta(
     SkeletalMeshComponent=@n1
     BoneName=clavicle_r
     BaseName=spine_05
     DeltaComponentType=QuaternionTwist
     InputRange=(X=2,Y=-20)
     TwistAxis="0,0,1"
    )) -> (Output_Get:double)
   > Set shoulderAlpha_l(shoulderAlpha_l:double=GetAlphaFromReferenceTransformDelta(
     SkeletalMeshComponent=@n1
     BoneName=clavicle_l
     BaseName=spine_05
     DeltaComponentType=QuaternionTwist
     InputRange=(X=2,Y=-20)
     TwistAxis="0,0,1"
    )) -> (Output_Get:double)
   n2 = Event OnBeginSkydiving()
   n3 = Get Dynamic Material Instance(execute=@n4.then, SkeletalMeshComponent=GetPartSkeletalMeshComponent())
   SetScalarParameterValue(execute=[@n2.then, @n6.else, @n7.then, @n8.else], self=@n3, ParameterName=Hide Element 01, Value=0)
   n4 = Event OnWeaponChanged()
   Event OnEndSkydiving()
   n5 = PickaxeItemDefinitionHasAnyMetaTag(Tags=Make Array("[0]"=(TagName="Cosmetics.MetaTags.TyphoonFemale")), bRequiresEquipped=true)
   > n6 = Branch(Condition=@n5)
   SetScalarParameterValue(execute=[@n6.then, @n7.else], self=@n3, ParameterName=Hide Element 01, Value=1)
   n7 = Branch(execute=@n8.then, Condition=Get bIsSkyDiving())
   n8 = Branch(execute=@n3.then, Condition=@n5)
  ]}
]
variables [
 inflateAlpha_r {type=number}
 shoulderAlpha_l {type=number}
 shoulderAlpha_r {type=number}
]
